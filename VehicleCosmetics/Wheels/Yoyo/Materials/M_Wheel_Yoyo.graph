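# Masked material for the Yoyo vehicle wheel cosmetic, layering rim and tire effects over MF_WheelParent_Simple.
# Vertex red separates tire from rim, whose base color blends paint tint, panning finish-line texture and fake cubemap reflections through masks.
# Emissive combines rim glow with switchable tire light; a UV-warped lava flipbook supplies a named alpha.
# Optional LOD opacity masking is set before the shared damage and de-rez transition functions.
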Material /VehicleCosmetics/Wheels/Yoyo/Materials/M_Wheel_Yoyo
graph {
 properties {
  bAutomaticallySetUsageInEditor=false
  bUseMaterialAttributes=true
  bUsedWithMeshDeformer=true
  bUsedWithSkeletalMesh=true
  blendMode=BLEND_Masked
  pixelDepthOffsetMode=PDOM_Legacy
  previewMesh=/DealershipAntennas/Wolvo/Meshes/SM_Antenna_WolvoTire
  usedShadingModels=Default Lit
 }
 nodes [
  n1 = MaterialFunctionCall() {materialFunction=&MF_WheelParent_Simple}
  n2 = GetMaterialAttributes(None=@n1.MaterialAttributes) {attributeGetTypes [#1, #2]}
  n3 = NamedRerouteUsage()
  n4 = TextureCoordinate()
  n5 = TextureSample(UVs=Multiply(A=Panner(Coordinate=@n4) {speedY=0.1}, B=Constant2Vector() {g=2, r=2})) {
   samplerSource=SSM_Wrap_WorldGroupSettings
   samplerType=SAMPLERTYPE_Grayscale
   texture=/VehicleCosmetics/Wheels/Yoyo/Texture/T_Yoyo_Finishlines
  }
  n6 = ConstantBiasScale(None=@n5.RGB) {bias=0, scale=0.2}
  n7 = VectorParameter() {parameterName=PaintColor, defaultValue=<rgba 1 0.4 0 1>}
  n8 = Constant() {r=0.25}
  n9 = Multiply(A=Constant() {r=25}, B=ConstantBiasScale(None=@n5.RGB) {bias=-0.25, scale=3})
  n10 = NamedRerouteUsage()
  n11 = Constant() {r=1.2}
  n12 = TextureSample(UVs=Transform(None=ReflectionVectorWS())) {
   samplerSource=SSM_Wrap_WorldGroupSettings
   samplerType=SAMPLERTYPE_LinearColor
   texture=/VehicleCosmetics/Wheels/DuelingDragonsSE/Materials/TC_DuelingDragonsSE_GrassCube
  }
  n13 = ConstantBiasScale(None=Multiply(A=@n12.RGB, B=ScalarParameter() {parameterName=ENVIntensity})) {bias=-0.2, scale=3}
  n14 = TextureSampleParameter2D() {parameterName=BC_map, samplerType=SAMPLERTYPE_LinearColor, texture=/VehicleCosmetics/Wheels/Yoyo/Texture/T_Yoyo_Rim_BC}
  n15 = Multiply(A=Multiply(A=Multiply(A=@n10, B=@n11), B=@n7.RGB), B=Add(A=Clamp(None=@n13), B=@n14.A))
  n16 = LinearInterpolate(
   A=LinearInterpolate(A=Multiply(A=@n6, B=@n7.RGB), B=Multiply(A=@n8, B=@n7.RGB), Alpha=Clamp(None=@n9))
   B=Multiply(A=@n15, B=@n15)
   Alpha=NamedRerouteUsage()
  )
  n17 = Add(A=@n16, B=VectorParameter() {parameterName=PaintColor_blend, defaultValue=<rgba 0.375 0.1688 0.2108 1>}.RGB) {constB=0.1}
  n18 = TextureSampleParameter2D() {
   parameterName=Effect_Mask
   samplerSource=SSM_Wrap_WorldGroupSettings
   samplerType=SAMPLERTYPE_LinearColor
   texture=/VehicleCosmetics/Wheels/Yoyo/Texture/T_Yoyo_Rim_RGB
  }
  n19 = NamedRerouteUsage()
  n20 = Multiply(A=Add(A=@n12.RGB, B=@n12.RGB), B=@n19)
  n21 = TextureSample(UVs=Transform(None=ReflectionVectorWS())) {
   samplerSource=SSM_Wrap_WorldGroupSettings
   samplerType=SAMPLERTYPE_LinearColor
   texture=/VehicleCosmetics/Mutable/FX_Materials/WorldEnv/WorldEnv_AthenaCubemap
  }
  n22 = ConstantBiasScale(None=Clamp(None=ConstantBiasScale(None=@n21.R) {bias=-0.5, scale=3})) {bias=0.25, scale=1.75}
  n23 = Multiply(A=Add(A=Multiply(A=@n21.R, B=@n22), B=ScalarParameter() {defaultValue=2, parameterName=ReflectionBrightness}), B=@n19)
  n24 = LinearInterpolate(
   A=NamedRerouteDeclaration(None=@n2.Base Color) {name=BaseColor, nodeColor=<rgba 0.5234 1 0 1>}
   B=LinearInterpolate(
    A=LinearInterpolate(A=@n20, B=@n23)
    B=StaticSwitchParameter(True=@n16, False=@n17) {defaultValue=true, parameterName=LOD4_Color}
    Alpha=@n18.R
   )
   Alpha=OneMinus(None=@n1.VertexColorR)
  )
  n25 = TextureSampleParameter2D() {
   parameterName=Alpha_LOD
   samplerSource=SSM_Wrap_WorldGroupSettings
   samplerType=SAMPLERTYPE_LinearColor
   texture=/VehicleCosmetics/Wheels/Yoyo/Texture/T_Yoyo_Tier_M_LOD
  }
  n26 = StaticSwitchParameter(True=LinearInterpolate(A=@n25.R, B=@n25.R, Alpha=NamedRerouteUsage()), False=Constant() {r=1}) {parameterName=Opacity_LOD}
  MaterialFunctionCall(In=MaterialFunctionCall(Material=SetMaterialAttributes!(MaterialAttributes=@n2.MaterialAttributes, Emissive Color=@n3, Base Color=@n24, Opacity Mask=@n26) {attributeSetTypes [#1, #2, #3]}) {materialFunction=&MF_BaseVehicle_Damage}.Result) {materialFunction=&MF_VehicleDeRez_Transition}
  n27 = Constant()
  n28 = Constant() {r=16}
  n29 = NamedRerouteUsage()
  n30 = NamedRerouteUsage()
  n31 = Multiply(
   A=Clamp(None=ConstantBiasScale(None=@n30) {bias=0, scale=1}) {minDefault=0.1}
   B=ScalarParameter() {defaultValue=1, parameterName=VFX_Emissive}
  )
  n32 = Add(
   A=LinearInterpolate(A=@n29, B=Multiply(A=@n28, B=@n29), Alpha=ConstantBiasScale(None=@n5.RGB) {bias=-0.3, scale=1})
   B=Multiply(A=Multiply(A=@n31, B=@n29), B=ScalarParameter() {defaultValue=6, parameterName=RimEmissiveAmount})
  )
  n33 = NamedRerouteUsage()
  n34 = VectorParameter() {parameterName=PaintColor, defaultValue=<rgba 1 0.01 0 1>}
  n35 = TextureSample() {samplerSource=SSM_Wrap_WorldGroupSettings, samplerType=SAMPLERTYPE_LinearColor, texture=/VehicleCosmetics/Wheels/Yoyo/Texture/T_Yoyo_Tire_M}
  n36 = ScalarParameter() {defaultValue=2, parameterName=EmissiveTire}
  n37 = Multiply(A=LinearInterpolate(A=@n27, B=@n34.RGB, Alpha=@n35.G), B=@n36)
  n38 = VertexColor()
  NamedRerouteDeclaration(None=LinearInterpolate(
    A=LinearInterpolate(A=@n27, B=@n32, Alpha=@n33)
    B=StaticSwitchParameter(True=@n37, False=@n36) {defaultValue=true, parameterName=TireLight}
    Alpha=@n38.R
   )) {name=Emissive, nodeColor=<rgba 1 0.75 0 1>}
  n39 = TextureCoordinate()
  n40 = ScalarParameter() {defaultValue=1, parameterName=Tiling}
  n41 = Panner(Coordinate=Frac(None=Multiply(A=@n39, B=@n40))) {speedY=0.2}
  n42 = Multiply(
   A=ComponentMask(None=TextureSample(UVs=@n41) {samplerSource=SSM_Wrap_WorldGroupSettings, texture=/VehicleCosmetics/Wheels/_SharedTextures/FX/Distort_RGB_T}.RGB) {g=true, r=true}
   B=ScalarParameter() {defaultValue=1, parameterName=Warp}
  )
  n43 = AppendVector(A=Constant() {r=8}, B=Constant() {r=8})
  n44 = Multiply(A=@n40, B=TextureCoordinate() {coordinateIndex=1})
  NamedRerouteDeclaration(None=MaterialFunctionCall!(
    Texture=TextureObject() {texture=/VehicleCosmetics/Wheels/Yoyo/Texture/T_Yoyo_Flipbook_HotLava_8x8}
    SubImages=Add(A=@n42, B=@n43)
    Speed=ScalarParameter() {defaultValue=0.5, parameterName=Speed}
    UVs=Frac(None=@n44)
    Offset_Frame_Position
    Linear_Blend=StaticBool() {value=true}
   ) {materialFunction=/VehicleCosmetics/Wheels/Mild/New/Materials/Skin_SubUV}.Output_Alpha) {name=OutputAplha_Scin_SubUV, nodeColor=<rgba 0.0078 1 0 1>}
  NamedRerouteDeclaration(None=@n34.RGB) {name=PaintColor, nodeColor=<rgba 1 0 0.3281 1>}
  NamedRerouteDeclaration(None=@n14.RGB) {name=Texture_PaintFinish, nodeColor=<rgba 0 0.6484 1 1>}
  n45 = TextureSampleParameter2D() {
   parameterName=Effecr_Mask2
   samplerSource=SSM_Wrap_WorldGroupSettings
   samplerType=SAMPLERTYPE_LinearColor
   texture=/VehicleCosmetics/Wheels/Yoyo/Texture/T_Yoyo_Rim_M
  }
  NamedRerouteDeclaration(None=@n45.G) {name=EffectMask_Green, nodeColor=<rgba 0.3125 0 1 1>}
  NamedRerouteDeclaration(None=@n38.R) {name=VertexColor, nodeColor=<rgba 1 0.6094 0 1>}
  NamedRerouteUsage()
 ]
}
