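# Masked, skeletal-mesh material for the Yoyo cosmetic vehicle wheel, showing an animated, paint-tinted flipbook effect.
# Skin_SubUV plays an 8x8 flipbook texture over tiled, wrapped UVs with linear frame blending.
# The desaturated, biased and clamped result is tinted by PaintColor for base color and emissive, while the raw flipbook, scaled by a strength parameter, drives opacity mask.
# Shared vehicle damage and DeRez transition functions then process these attributes.
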Material /VehicleCosmetics/Wheels/Yoyo/Materials/M_Wheel_Yoyo_SE2
graph {
 properties {
  bAutomaticallySetUsageInEditor=false
  bUseMaterialAttributes=true
  bUsedWithSkeletalMesh=true
  blendMode=BLEND_Masked
  usedShadingModels=Default Lit
 }
 nodes [
  n1 = TextureCoordinate()
  n2 = ScalarParameter() {defaultValue=1, parameterName=Tiling}
  n3 = MaterialFunctionCall!(
   Texture=TextureObject() {samplerType=SAMPLERTYPE_LinearColor, texture=/VehicleCosmetics/Wheels/Yoyo/Texture/T_Yoyo_Flipbook_3_8x8}
   SubImages=AppendVector(A=Constant() {r=8}, B=Constant() {r=8})
   Speed=ScalarParameter() {defaultValue=0.4, parameterName=Speed}
   UVs=Frac(None=Multiply(A=@n1, B=@n2))
   Offset_Frame_Position
   Linear_Blend=StaticBool() {value=true}
  ) {materialFunction=/VehicleCosmetics/Wheels/Mild/New/Materials/Skin_SubUV}
  n4 = Desaturation(None=@n3.Output_RGB, Fraction=Constant() {r=1})
  n5 = Multiply(
   A=Clamp(None=ConstantBiasScale(None=@n4) {bias=0.2, scale=2})
   B=VectorParameter() {parameterName=PaintColor, defaultValue=<rgba 1 0.09 0.13 1>}.RGB
  )
  n6 = Multiply(A=ScalarParameter() {defaultValue=24, parameterName=Strengh_Mask}, B=@n3.Output_RGB)
  MaterialFunctionCall(In=MaterialFunctionCall(Material=SetMaterialAttributes!(MaterialAttributes, Base Color=@n5, Emissive Color=@n5, Opacity Mask=@n6) {attributeSetTypes [#1, #2, #3]}) {materialFunction=&MF_BaseVehicle_Damage}.Result) {materialFunction=&MF_VehicleDeRez_Transition}
 ]
}
