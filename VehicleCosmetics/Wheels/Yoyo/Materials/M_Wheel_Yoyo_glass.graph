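# Unlit additive glass material for the Yoyo vehicle wheel cosmetic.
# Panned finish-line texture, tint and reflection-sampled environment texture sum with paint-colored Fresnel.
# Rim diffuse alpha masks that sum, then rim RGB blue blends it toward flat PaintColor emissive.
# Emissive and scalar opacity pass through vehicle damage and de-rez transition functions; the scaled rim normal branch is unconnected.
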Material /VehicleCosmetics/Wheels/Yoyo/Materials/M_Wheel_Yoyo_glass
graph {
 properties {
  bAutomaticallySetUsageInEditor=false
  bUseMaterialAttributes=true
  bUsedWithMeshDeformer=true
  bUsedWithSkeletalMesh=true
  blendMode=BLEND_Additive
  shadingModel=MSM_Unlit
  usedShadingModels=Unlit
 }
 nodes [
  n1 = TextureCoordinate()
  n2 = TextureSample(UVs=Multiply(A=Panner(Coordinate=@n1) {speedY=0.1}) {constB=2.2}) {samplerType=SAMPLERTYPE_Grayscale, texture=/VehicleCosmetics/Wheels/Yoyo/Texture/T_Yoyo_Finishlines}
  n3 = VectorParameter() {parameterName=ENVTint, defaultValue=<rgba 0.06 0.2 0.28 1>}
  n4 = ReflectionVectorWS()
  n5 = Multiply(
   A=TextureSample(UVs=Transform(None=@n4)) {samplerType=SAMPLERTYPE_LinearColor, texture=/Game/Creative/Environments/Meshes/Glass/Textures/T_Glass_Cubemap_01}.RGB
   B=ScalarParameter() {defaultValue=2, parameterName=ENVIntensity}
  )
  n6 = Add(A=Add(A=Add(A=@n2.RGB, B=@n3.RGB)), B=Clamp(None=ConstantBiasScale(None=@n5) {bias=-0.2, scale=3}))
  n7 = MaterialFunctionCall!(Normal, Exponent=Constant() {r=1}) {materialFunction=/VehicleCosmetics/Wheels/Cola_Holographic/Materials/MF_Fresnel_Classic}
  n8 = Multiply(
   A=ScalarParameter() {defaultValue=2, parameterName=EmissiveEmount}
   B=VectorParameter() {parameterName=PaintColor, defaultValue=<rgba 0.2 0.7 0.95 1>}.RGB
  )
  n9 = LinearInterpolate(A=@n7.Result, B=Multiply(A=@n7.Result, B=@n8), Alpha=Constant() {r=1})
  n10 = LinearInterpolate(
   A=Multiply(
    A=Add(A=@n6, B=@n9)
    B=TextureSampleParameter2D() {parameterName=Rim_Difffuse, samplerType=SAMPLERTYPE_LinearColor, texture=/VehicleCosmetics/Wheels/Yoyo/Texture/T_Yoyo_Rim_BC}.A
   )
   B=@n8
   Alpha=TextureSampleParameter2D() {parameterName=Rim_RGB, samplerType=SAMPLERTYPE_LinearColor, texture=/VehicleCosmetics/Wheels/Yoyo/Texture/T_Yoyo_Rim_RGB}.B
  )
  n11 = ScalarParameter() {defaultValue=0.02, parameterName=Opacity}
  MaterialFunctionCall(In=MaterialFunctionCall(Material=SetMaterialAttributes!(MaterialAttributes, Emissive Color=@n10, Opacity=@n11) {attributeSetTypes [#1, #2]}) {materialFunction=&MF_BaseVehicle_Damage}.Result) {materialFunction=&MF_VehicleDeRez_Transition}
  Multiply(
   A=TextureSampleParameter2D() {parameterName=Rim_Normal, samplerType=SAMPLERTYPE_Normal, texture=/VehicleCosmetics/Wheels/Yoyo/Texture/T_Yoyo_Rim_N2}.RGB
   B=Constant3Vector() {constant=<rgba 0.625 0.625 1 0>}
  )
 ]
}
